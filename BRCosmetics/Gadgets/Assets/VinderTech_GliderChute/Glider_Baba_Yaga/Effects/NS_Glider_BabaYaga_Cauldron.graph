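# Niagara effect for the Baba Yaga glider's cauldron: bubbles, flames, embers, fill and decal sprites, plus flickering lights.
# Every emitter samples the user skeletal mesh and passes particle positions through TransformToBoneSpace to follow a bone.
# Front-end flags adjust that transform, and most spawn offsets have separate gameplay and front-end values.
# Flames and embers react to owner velocity, with TrailLengthScale driving ember rate, lifetime and speed limits.
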NiagaraSystem /BRCosmetics/Gadgets/Assets/VinderTech_GliderChute/Glider_Baba_Yaga/Effects/NS_Glider_BabaYaga_Cauldron
dependencies {
 usedDataInterfaces [&NiagaraDataInterfaceCurve, &NiagaraDataInterfaceColorCurve, &NiagaraDataInterfaceVectorCurve]
 usedDynamicInputs [&VectorFromCurve]
 usedModules [&SystemState]
}
emitters {
 Bubbles {
  emitterUpdateScript [
   &EmitterState()
   &SpawnRate(
    Spawn Probability=0.9
    SpawnRate=&UniformRangedFloat(Minimum=200, Maximum=350)
    Use Spawn Probability=true
   )
  ]
  particleSpawnScript [
   /SetVariables(
    Particles.CylinderRadius=10
    Particles.OriginPoint:Vector3f=&FN_FrontEndVector3Modifier(Gameplay=<0 -9 22>, Front End=<0 -9 12>)
    Particles.PickScale:NiagaraFloat=&UniformRangedFloat(
     Minimum=&RandomRangeFloat(Minimum=0.125, Maximum=1)
     Maximum=&RandomRangeFloat(Minimum=2, Maximum=5)
    )
    Particles.RandomPosNeg:NiagaraFloat=&UniformRangedFloat(Minimum=-1, Maximum=1)
   ) {isSetParameters=true}
   /Niagara/Modules/Spawn/Initialization/InitializeParticle(
    Color=<rgba 1 1 2 1>
    Lifetime=&Lerp_Float(
     A=0.05
     B=&RandomRangeFloat(Minimum=0.2, Maximum=0.4)
     Alpha=&Multiply_Float(A=0.333, B=$Particles.PickScale)
    )
    Mesh Scale=&VectorFromFloat(Value=10)
    Module.Write Color=true
    Module.Write Lifetime=true
    Module.Write Mass=false
    Module.Write Material Random=true
    Module.Write Position=false
    Module.Write Scale=false
    Module.Write SpriteRotation=false
    Module.Write SpriteSize=true
    Position=&SimulationPosition()
    Sprite Size=&Vector2DFromFloat(Value=$Particles.PickScale)
   )
   &AddVelocity(Velocity=&UniformRangedVector(Minimum=<-10 -10 10>, Maximum=<10 10 15>))
   &CylinderLocation(
    Cylinder Height=0
    Cylinder Origin:Vector3f=$Particles.OriginPoint
    Cylinder Radius=$Particles.CylinderRadius
    MaskSpawnGroup=false
    Module.Enable Lathe Profile=false
    Module.Override Randomness=false
    Module.Surface Only=false
   ) {module=CylinderLocation001}
  ]
  particleUpdateScript [
   &ParticleState()
   &SkeletalMeshLocation(
    Apply To Particle Position=false
    Bone Sampling Mode=::NewEnumerator4
    Inherit Velocity=false
    MaskSpawnGroup=false
    Skeletal Mesh=$User.SkeletalMesh
    Use Sampled Mesh Attribute as Ribbon ID=false
   )
   /SetVariables(Particles.BoneLocation:Vector3f=$Output.SkeletalMeshLocation.SampledPosition) {isSetParameters=true}
   /SetVariables(Particles.DistanceFromCenter:NiagaraFloat=&ClampFloat(Float=&Add_Float(
      A=-1
      B=&Multiply_Float(
       A=0.2
       B=&VectorLength(Vector=&Multiply_Vector(
         A=&Subtract_Vector(A=Linked() {linkedVariable {name=Particles.Position, type=NiagaraPosition}}, B=$Particles.OriginPoint)
         B=<1 1 0>
        ))
      )
     ))) {isSetParameters=true}
   &ScaleVelocity(Velocity Scale=&VectorFromFloat(Value=&Multiply_Float(
      A=&Lerp_Float(A=1, B=0.2, Alpha=$Particles.NormalizedAge)
      B=&ClampFloat(Float=&OneMinusFloat(Float=&Multiply_Float(A=$Particles.DistanceFromCenter, B=$Particles.DistanceFromCenter)))
     )))
   &SolveForcesAndVelocity(Clamp Velocity=false, Limit Acceleration=false)
   /Niagara/Modules/Update/Orientation/V2/UpdateMeshOrientation(Rotation Rate:NiagaraFloat=$Particles.RandomPosNeg, Rotation Vector:Vector3f=&VectorFromFloat(Value=1)) {enabled=false}
   /Niagara/Modules/Update/Orientation/V3/UpdateMeshOrientation(Rotation Rate=$Particles.RandomPosNeg, Rotation Vector=&VectorFromFloat(Value=1)) {module=UpdateMeshOrientation001}
   &ScaleSpriteSize(Scale Factor=&Vector2DFromFloat(Value=&FloatFromCurve(FloatCurve=DataInterface() {propertyValues {Curve {keys [
          <at 0 0 RCIM_Cubic>
          {arriveTangent=0.002, interpMode=RCIM_Cubic, leaveTangent=-0.0489, tangentMode=RCTM_Break, time=0.25, value=0.9069}
          {arriveTangent=-0.077, interpMode=RCIM_Cubic, leaveTangent=-0.077, tangentMode=RCTM_User, time=0.9114, value=0.8888}
          {arriveTangent=4.9648, interpMode=RCIM_Cubic, leaveTangent=4.9648, tangentMode=RCTM_User, time=1, value=1.2}
         ]}}})))
   &DynamicMaterialParameters(
    Index 0 Param 1=1
    Index 0 Param 2=1
    Index 0 Param 3=1
    Module.Param0WriteEnabled=true
    Module.Param1WriteEnabled=true
    Module.Param2WriteEnabled=true
    Module.Param3WriteEnabled=true
   )
   &self:TransformToBoneSpace(
    BonePosition:Vector3f=$Particles.BoneLocation
    BoneRotation:Quat4f=$Output.SkeletalMeshLocation.SampledRotation
    Front End Toggle:NiagaraBool=$User.bIsFrontEnd
    Position:NiagaraPosition=$Particles.Position
   ) {module=TransformToBoneSpace}
  ]
  rendererClasses [&NiagaraSpriteRendererProperties, &NiagaraMeshRendererProperties]
  renderers [
   {rendererClass=&NiagaraSpriteRendererProperties}
   {rendererClass=&NiagaraMeshRendererProperties, rendererIndex=1}
  ]
 }
 Decal {
  emitterUpdateScript [
   &EmitterState()
   &SpawnRate(SpawnRate=15, Use Spawn Probability=false)
  ]
  particleSpawnScript [
   &InitializeParticle(
    Color=&MakeLinearColor(
     R:NiagaraFloat=1
     G:NiagaraFloat=0
     B:NiagaraFloat=0
     A:NiagaraFloat=&RandomRangeFloat(Minimum=1, Maximum=1)
    )
    Lifetime=&UniformRangedFloat(Minimum=0.25, Maximum=0.5)
    Module.Write Color=true
    Module.Write Lifetime=true
    Module.Write Mass=false
    Module.Write Material Random=true
    Module.Write Position=false
    Module.Write Scale=false
    Module.Write SpriteRotation=false
    Module.Write SpriteSize=true
    Position=&SimulationPosition()
    Sprite Rotation=&UniformRangedFloat(Minimum=-180, Maximum=180)
    Sprite Size=&MakeVector2D(
     X=&RandomRangeFloat(Minimum=0.5, Maximum=1)
     Y=&RandomRangeFloat(Minimum=0.5, Maximum=0.5)
    )
   )
   &ApplyInitialForces()
   &DynamicMaterialParameters(
    Index 0 Param 1=0.5
    Index 0 Param 2=1
    Index 0 Param 3=1
    Index 0 Param 4=&RandomRangeFloat(Minimum=0, Maximum=1)
    Module.Param0WriteEnabled=true
    Module.Param1WriteEnabled=true
    Module.Param2WriteEnabled=true
    Module.Param3WriteEnabled=true
   )
   &CylinderLocation(
    Cylinder Height=0
    Cylinder Radius=10
    Lathe Profile=DataInterface() {propertyValues {Curve {keys [
        {arriveTangent=1.3231, interpMode=RCIM_Cubic, leaveTangent=1.3231, tangentMode=RCTM_User, time=0, value=0}
        {interpMode=RCIM_Cubic, leaveTangent=0.049, time=1, value=1}
       ]}}}
    MaskSpawnGroup=false
    Module.Enable Lathe Profile=false
    Module.Override Randomness=false
    Module.Surface Only=false
    Offset=&FN_FrontEndVector3Modifier(Gameplay=<0 -9 2>, Front End=<0 -9 -8>)
    Surface Only Band Thickness=2
   )
   &AddVelocityFromPoint(Module.Use Falloff Distance=false, Velocity Strength=5)
   /SetVariables(Particles.Velocity=&Multiply_Vector(A=$Particles.Velocity, B=&Lerp_Vector(A=<1 1 0.1>, B=<1 1 10>, Alpha=$Particles.Lifetime))) {isSetParameters=true}
   /SetVariables(Particles.StableRandom:NiagaraFloat=&RandomRangeFloat(Minimum=0, Maximum=1)) {isSetParameters=true}
  ]
  particleUpdateScript [
   &ParticleState()
   &SkeletalMeshLocation(
    Apply To Particle Position=false
    Bone Sampling Mode=::NewEnumerator4
    Inherit Velocity=false
    MaskSpawnGroup=false
    Skeletal Mesh=$User.SkeletalMesh
    Use Sampled Mesh Attribute as Ribbon ID=false
   )
   /SetVariables(Particles.BoneLocation:Vector3f=$Output.SkeletalMeshLocation.SampledPosition) {isSetParameters=true}
   &SolveForcesAndVelocity(Clamp Velocity=false, Limit Acceleration=false)
   &ScaleSpriteSize(Scale Factor=&Vector2DFromFloat(Value=&FloatFromCurve(
      FloatCurve=DataInterface() {propertyValues {Curve {keys [
          {arriveTangent=0.588, interpMode=RCIM_Cubic, leaveTangent=0.588, tangentMode=RCTM_User, time=0, value=0.9}
          {arriveTangent=0.111, interpMode=RCIM_Cubic, leaveTangent=0.111, tangentMode=RCTM_User, time=0.18, value=0.99}
          {arriveTangent=-0.469, interpMode=RCIM_Cubic, leaveTangent=-0.469, tangentMode=RCTM_User, time=1, value=0.75}
         ]}}}
      Scale Curve=&Lerp_Float(A=10, B=25, Alpha=$Particles.StableRandom)
     )))
   &CameraOffset(Camera Offset Amount=15)
   &ScaleColor(
    Scale Alpha=&FloatFromCurve(FloatCurve=DataInterface() {propertyValues {Curve {keys [
         {arriveTangent=3.127, interpMode=RCIM_Cubic, leaveTangent=3.127, tangentMode=RCTM_User, time=0, value=0}
         {interpMode=RCIM_Cubic, tangentMode=RCTM_User, time=0.5, value=1}
         {arriveTangent=-3.2738, interpMode=RCIM_Cubic, leaveTangent=-3.2738, tangentMode=RCTM_User, time=1, value=0}
        ]}}})
    ScaleA=true
    ScaleRGB=false
   )
   &self:TransformToBoneSpace(
    BonePosition:Vector3f=$Particles.BoneLocation
    BoneRotation:Quat4f=$Output.SkeletalMeshLocation.SampledRotation
    Front End Toggle:NiagaraBool=$User.bIsFrontEnd
    Position:NiagaraPosition=$Particles.Position
   ) {module=TransformToBoneSpace}
  ]
  rendererClasses [&NiagaraSpriteRendererProperties]
  renderers [
   {rendererClass=&NiagaraSpriteRendererProperties}
   {rendererClass=&NiagaraSpriteRendererProperties, rendererIndex=1}
  ]
 }
 Embers {
  emitterUpdateScript [
   &EmitterState()
   &SpawnRate(
    Spawn Probability=0.2
    SpawnRate=&Lerp_Float(A=40, B=100, Alpha=$User.TrailLengthScale)
    Use Spawn Probability=true
   )
  ]
  particleSpawnScript [
   &InitializeParticle(
    Color=<rgba 2 1.2022 0.6482 1>
    Lifetime=&UniformRangedFloat(Minimum=0.1, Maximum=&Lerp_Float(A=0.25, B=1, Alpha=$User.TrailLengthScale))
    Module.Write Color=true
    Module.Write Lifetime=true
    Module.Write Mass=false
    Module.Write Material Random=true
    Module.Write Position=false
    Module.Write Scale=false
    Module.Write SpriteRotation=true
    Module.Write SpriteSize=true
    Position=&SimulationPosition()
    Sprite Rotation=&UniformRangedFloat(Minimum=-180, Maximum=180)
    Sprite Size=&Multiply_Vector2(
     A=&MakeVector2D(
      X=&UniformRangedFloat(Minimum=0.5, Maximum=1)
      Y=&UniformRangedFloat(Minimum=1, Maximum=3)
     )
     B=&Vector2DFromFloat(Value=&RandomRangeFloat(Minimum=1, Maximum=2))
    )
   )
   &AddVelocity(Velocity=&MakeVector(
     X=&UniformRangedFloat(Minimum=-50, Maximum=50)
     Y=10
     Z=&UniformRangedFloat(Minimum=100, Maximum=200)
    ))
   &ApplyInitialForces()
   &BoxLocation(
    Box Offset=&FN_FrontEndVector3Modifier(Gameplay=<0 10 0>, Front End=<0 10 -10>)
    Box Size=<15 2 0>
    MaskSpawnGroup=false
    Module.Override Seed=false
    Rotate Box=false
    Surface Only=false
   )
  ]
  particleUpdateScript [
   &ParticleState()
   &SkeletalMeshLocation(
    Apply To Particle Position=false
    Bone Sampling Mode=::NewEnumerator4
    Inherit Velocity=false
    MaskSpawnGroup=false
    Skeletal Mesh=$User.SkeletalMesh
    Use Sampled Mesh Attribute as Ribbon ID=false
   )
   /SetVariables(Particles.BoneLocation:Vector3f=$Output.SkeletalMeshLocation.SampledPosition) {isSetParameters=true}
   &AddVelocity(Velocity=&RandomRangeVector(Minimum=<-50 -10 -10>, Maximum=<50 40 10>, Evaluation Type=::NewEnumerator1)) {module=AddVelocity001}
   /SetVariables(Particles.Velocity=&self:ClampVector(
     Max:Vector3f=&Lerp_Vector(A=<100 50 1000>, B=<200 1000 200>, Alpha=$User.TrailLengthScale)
     Min:Vector3f=<-100 -100 -100>
     Vector:Vector3f=$Particles.Velocity
    )) {isSetParameters=true}
   &InheritVelocity(
    Module.Limit Speed=true
    Velocity Limit:NiagaraFloat=&Lerp_Float(A=50, B=200, Alpha=$User.TrailLengthScale)
    Velocity Scale=<-0.1 -0.1 0>
    Velocity Source=&TransformVector(
     Vector=&RotateVectorByQuaternion(Vector To Rotate=$Engine.Owner.Velocity, Quaternion=$Particles.BoneOrient)
     Source Space=::World
     Destination Space=::Simulation
    )
   )
   &AccelerationForce(Acceleration=<0 0 50>)
   &Drag(Drag=3, Use Linear Drag=true, Use Rotational Drag=false)
   &SolveForcesAndVelocity(Clamp Velocity=false, Limit Acceleration=false)
   &ScaleColor(
    Scale Alpha=&FloatFromCurve(FloatCurve=DataInterface() {propertyValues {Curve {keys [
         {arriveTangent=11.9433, interpMode=RCIM_Cubic, leaveTangent=11.9433, tangentMode=RCTM_User, time=0, value=0}
         <at 0.2 1 RCIM_Cubic>
         <at 1 0 RCIM_Cubic>
        ]}}})
    Scale RGB=&MakeVectorFromLinearColorRGB(Color=&ColorFromCurve(ColorCurve=DataInterface() {propertyValues {
        AlphaCurve {keys [<at 0 1 RCIM_Cubic>]}
        BlueCurve {keys [<at 0 0.5069 RCIM_Cubic>,<at 1 0 RCIM_Cubic>]}
        GreenCurve {keys [<at 0 1.6422 RCIM_Cubic>,<at 1 0.06 RCIM_Cubic>]}
        RedCurve {keys [<at 0 4 RCIM_Cubic>,<at 1 1 RCIM_Cubic>]}
       }}))
    ScaleA=true
    ScaleRGB=true
   )
   &DynamicMaterialParameters(
    Index 0 Param 1=&FloatFromCurve(FloatCurve=DataInterface() {propertyValues {Curve {keys [<at 0 0>,<at 1 1>]}}})
    Index 0 Param 2=1
    Index 0 Param 3=1
    Index 0 Param 4=1
    Module.Param0WriteEnabled=true
    Module.Param1WriteEnabled=true
    Module.Param2WriteEnabled=true
    Module.Param3WriteEnabled=true
   )
   &ScaleSpriteSize(Scale Factor=&Vector2DFromFloat(Value=&FloatFromCurve(
      FloatCurve=DataInterface() {propertyValues {Curve {keys [
          {arriveTangent=9.1531, interpMode=RCIM_Cubic, leaveTangent=9.1531, tangentMode=RCTM_User, time=0, value=0.02}
          {arriveTangent=0.48, interpMode=RCIM_Cubic, leaveTangent=0.48, time=0.18, value=0.99}
          <at 1 0.5 RCIM_Cubic>
         ]}}}
      Scale Curve=&ClampFloat(Float=&OneMinusFloat(Float=&Multiply_Float(
         A=&VectorLength(Vector=&Subtract_Vector(
           A=Linked() {linkedVariable {name=Particles.Position, type=NiagaraPosition}}
           B=Linked() {linkedVariable {name=Particles.Initial.Position, type=NiagaraPosition}}
          ))
         B=&RandomRangeFloat(Minimum=0.0015, Maximum=0.002)
        )))
     )))
   &ScaleSpriteSizeBySpeed(Max Scale Factor=<1 4>)
   &KillParticlesInVolume(Box Size:Vector3f=<20 175 500>, Kill Shape=::NewEnumerator1, Origin Offset:Vector3f=<45 0 0>, Volume Origin:Vector3f=$Particles.OriginPoint)
   &KillParticlesInVolume(Box Size:Vector3f=<20 175 500>, Kill Shape=::NewEnumerator1, Origin Offset:Vector3f=<-45 0 0>, Volume Origin:Vector3f=$Particles.OriginPoint) {module=KillParticlesInVolume001}
   &self:TransformToBoneSpace(
    BonePosition:Vector3f=$Particles.BoneLocation
    BoneRotation:Quat4f=$Output.SkeletalMeshLocation.SampledRotation
    Front End Toggle:NiagaraBool=$User.bIsFrontEnd
    Position:NiagaraPosition=$Particles.Position
   ) {module=TransformToBoneSpace}
  ]
  rendererClasses [&NiagaraSpriteRendererProperties]
  renderers [{rendererClass=&NiagaraSpriteRendererProperties}]
 }
 Fill {
  emitterUpdateScript [
   &EmitterState()
   &SpawnRate(SpawnRate=5, Use Spawn Probability=false)
  ]
  particleSpawnScript [
   &InitializeParticle(
    Color=<rgba 5 3.0056 1.6206 0.75>
    Lifetime=&UniformRangedFloat(Minimum=0.5, Maximum=1)
    Module.Write Color=false
    Module.Write Lifetime=true
    Module.Write Mass=false
    Module.Write Material Random=true
    Module.Write Position=false
    Module.Write Scale=false
    Module.Write SpriteRotation=false
    Module.Write SpriteSize=true
    Position=&SimulationPosition()
    Sprite Rotation=&UniformRangedFloat(Minimum=-180, Maximum=180)
    Sprite Size=&Vector2DFromFloat(Value=&UniformRangedFloat(Minimum=20, Maximum=30))
   )
   &ApplyInitialForces()
   &TorusLocation(
    Align Torus With Axis=false
    Handle Radius=2
    Large Radius=10
    MaskSpawnGroup=false
    Module.Override Randomness=false
    Torus Origin:NiagaraPosition=&Add_Vector(
     A=Linked() {linkedVariable {name=Particles.Position, type=NiagaraPosition}}
     B=&FN_FrontEndVector3Modifier(Gameplay=<0 -9 0>, Front End=<0 -9 -10>)
    )
   )
   &Color(Color=<rgba 1 1 1 0.5>)
   &DynamicMaterialParameters(
    Index 0 Param 1=0.5
    Index 0 Param 2=2.5
    Index 0 Param 3=&RandomRangeFloat(Minimum=0, Maximum=1)
    Index 0 Param 4=&RandomRangeFloat(Minimum=0, Maximum=1)
    Module.Param0WriteEnabled=true
    Module.Param1WriteEnabled=true
    Module.Param2WriteEnabled=true
    Module.Param3WriteEnabled=true
   )
  ]
  particleUpdateScript [
   &ParticleState()
   &SkeletalMeshLocation(
    Apply To Particle Position=false
    Bone Sampling Mode=::NewEnumerator4
    Inherit Velocity=false
    MaskSpawnGroup=false
    Skeletal Mesh=$User.SkeletalMesh
    Use Sampled Mesh Attribute as Ribbon ID=false
   )
   /SetVariables(Particles.BoneLocation:Vector3f=$Output.SkeletalMeshLocation.SampledPosition) {isSetParameters=true}
   &SolveForcesAndVelocity(Clamp Velocity=false, Limit Acceleration=false)
   &ScaleColor(
    Scale Mode=::NewEnumerator1
    Scale RGBA=/Niagara/DynamicInputs/UniformRange/UniformRangedVector4(
     Minimum:Vector4f=&Vector4FromCurve(Vector4Curve=DataInterface() {propertyValues {
        AlphaCurve {keys [<at 0.0024 0>,<at 0.1271 1>,<at 0.265 0.9976>,<at 0.9977 -0.0026>]}
        BlueCurve {keys [<at 0.3937 0.5>,<at 0.75 0>]}
        GreenCurve {keys [<at 0.3937 2.875>,<at 0.75 0>]}
        RedCurve {keys [<at 0.3937 10>,<at 0.75 10>]}
       }})
     Maximum:Vector4f=&Vector4FromCurve(Vector4Curve=DataInterface() {propertyValues {
        AlphaCurve {keys [<at 0.0024 0>,<at 0.1271 1>,<at 0.265 0.9976>,<at 0.9977 -0.0026>]}
        BlueCurve {keys [<at 0.3937 0.5>,<at 0.75 0>]}
        GreenCurve {keys [<at 0.3937 2.875>,<at 0.75 3.7633>]}
        RedCurve {keys [<at 0.3937 10>,<at 0.75 10>]}
       }})
     Randomness Mode:ENiagaraRandomnessMode=::NewEnumerator0
     Random Seed:NiagaraInt32=0
     Fixed Random Seed:NiagaraBool=false
    )
    ScaleRGBA=true
   ) {module=ScaleColor001}
   &ScaleSpriteSize(Scale Factor=&Vector2DFromFloat(Value=&FloatFromCurve(FloatCurve=DataInterface() {propertyValues {Curve {keys [
          {arriveTangent=7.5764, interpMode=RCIM_Cubic, leaveTangent=7.5765, tangentMode=RCTM_User, time=0, value=0.02}
          {arriveTangent=1.2063, interpMode=RCIM_Cubic, leaveTangent=1.2063, tangentMode=RCTM_User, time=0.18, value=0.99}
          <at 1 0.5 RCIM_Cubic>
         ]}}})))
   &self:TransformToBoneSpace(
    BonePosition:Vector3f=$Particles.BoneLocation
    BoneRotation:Quat4f=$Output.SkeletalMeshLocation.SampledRotation
    Front End Toggle:NiagaraBool=$User.bIsFrontEnd
    Position:NiagaraPosition=$Particles.Position
   ) {module=TransformToBoneSpace}
  ]
  rendererClasses [&NiagaraSpriteRendererProperties]
  renderers [{rendererClass=&NiagaraSpriteRendererProperties}]
 }
 FlameLights {
  emitterUpdateScript [
   &EmitterState()
   &SpawnRate(SpawnRate=50, Use Spawn Probability=false)
  ]
  particleSpawnScript [
   &InitializeParticle(
    Color=<rgba 5 1.5234 0.3646 1>
    Lifetime=&UniformRangedFloat(Minimum=0.2, Maximum=0.4)
    Module.Write Color=true
    Module.Write Lifetime=true
    Module.Write Mass=false
    Module.Write Material Random=true
    Module.Write Position=false
    Module.Write Scale=false
    Module.Write SpriteRotation=false
    Module.Write SpriteSize=false
    Position=&SimulationPosition()
   )
   /SetVariables(Particles.OriginPoint:Vector3f=&Add_Vector(
     A=Linked() {linkedVariable {name=Particles.Position, type=NiagaraPosition}}
     B=&FN_FrontEndVector3Modifier(Gameplay=<0 -9 2>, Front End=<0 -9 -8>)
    )) {isSetParameters=true}
   &TorusLocation(
    Align Torus With Axis=false
    Handle Radius=2
    Large Radius=10
    MaskSpawnGroup=false
    Module.Override Randomness=false
    Torus Origin:Vector3f=$Particles.OriginPoint
   )
   &AddVelocityFromPoint(
    Module.Use Falloff Distance=false
    Origin Offset=&Add_Vector(A=$Particles.OriginPoint, B=<0 0 -1>)
    Velocity Strength=20
   )
   /SetVariables(
    Particles.LightExponent=1
    Particles.ParticleRandom:NiagaraFloat=&RandomRangeFloat(Minimum=0, Maximum=1)
    Particles.Velocity=&Multiply_Vector(A=$Particles.Velocity, B=<1 1 0.1>)
   ) {isSetParameters=true, module=SetVariables_#1}
  ]
  particleUpdateScript [
   &ParticleState()
   &SkeletalMeshLocation(
    Apply To Particle Position=false
    Bone Sampling Mode=::NewEnumerator4
    Inherit Velocity=false
    MaskSpawnGroup=false
    Skeletal Mesh=$User.SkeletalMesh
    Use Sampled Mesh Attribute as Ribbon ID=false
   )
   /SetVariables(Particles.BoneLocation:Vector3f=$Output.SkeletalMeshLocation.SampledPosition) {isSetParameters=true}
   /SetVariables(Particles.LightRadius=&FloatFromCurve(
     FloatCurve=DataInterface() {propertyValues {Curve {keys [
         {interpMode=RCIM_Cubic, tangentMode=RCTM_User, time=0.2, value=1}
         {arriveTangent=-1.0847, interpMode=RCIM_Cubic, leaveTangent=-1.0847, tangentMode=RCTM_User, time=1, value=0}
        ]}}}
     Scale Curve=&RandomRangeFloat(Minimum=5, Maximum=25)
    )) {isSetParameters=true}
   &InheritVelocity(
    Module.Limit Speed=true
    Velocity Limit=8
    Velocity Scale=<-0.01 -0.01 0>
    Velocity Source=&TransformVector(Vector=$Engine.Owner.Velocity, Source Space=::World, Destination Space=::Simulation)
   )
   &AccelerationForce(Acceleration=&UniformRangedVector(Minimum=<0 0 50>, Maximum=<1 1 100>))
   &Drag(Use Linear Drag=true, Use Rotational Drag=false)
   &SolveForcesAndVelocity(Clamp Velocity=false, Limit Acceleration=false)
   &ScaleColor(
    Scale Mode=::NewEnumerator1
    Scale RGBA=&MakeVector4_FromFloat(Float=&FloatFromCurve(FloatCurve=DataInterface() {propertyValues {Curve {keys [{arriveTangent=4.6264, interpMode=RCIM_Cubic, leaveTangent=4.6264, tangentMode=RCTM_User, time=0, value=0}, <at 0.4 1>]}}}))
    ScaleRGBA=true
   )
   &KillParticlesInVolume(Invert Volume=true, Sphere Radius=40, Volume Origin:Vector3f=$Particles.OriginPoint)
   &self:TransformToBoneSpace(
    BonePosition:Vector3f=$Particles.BoneLocation
    BoneRotation:Quat4f=$Output.SkeletalMeshLocation.SampledRotation
    Front End Toggle:NiagaraBool=$User.bIsFrontEnd
    Position:NiagaraPosition=$Particles.Position
   ) {module=TransformToBoneSpace}
  ]
  rendererClasses [&NiagaraLightRendererProperties]
  renderers [{rendererClass=&NiagaraLightRendererProperties, rendererIndex=0}]
 }
 Flames {
  emitterUpdateScript [
   &EmitterState()
   &SpawnRate(SpawnRate=50, Use Spawn Probability=false)
  ]
  particleSpawnScript [
   &InitializeParticle(
    Color=/Niagara/DynamicInputs/TypeConversions/MakeLinearColorFromFloat(Value:NiagaraFloat=&RandomRangeFloat(Minimum=0.25, Maximum=1))
    Lifetime=&UniformRangedFloat(Minimum=0.3, Maximum=&RandomRangeFloat(Minimum=0.4, Maximum=1))
    Mesh Scale=&VectorFromFloat(Value=3)
    Module.Write Color=true
    Module.Write Lifetime=true
    Module.Write Mass=false
    Module.Write Material Random=true
    Module.Write Position=false
    Module.Write Scale=true
    Module.Write SpriteRotation=true
    Module.Write SpriteSize=true
    Position=&SimulationPosition()
    Sprite Rotation=&UniformRangedFloat(Minimum=-180, Maximum=180)
    Sprite Size=&Vector2DFromFloat(Value=&UniformRangedFloat(Minimum=20, Maximum=30))
   )
   /SetVariables(Particles.OriginPoint:Vector3f=&Add_Vector(
     A=Linked() {linkedVariable {name=Particles.Position, type=NiagaraPosition}}
     B=&FN_FrontEndVector3Modifier(Gameplay=<0 -9 2>, Front End=<0 -9 -8>)
    )) {isSetParameters=true}
   &TorusLocation(
    Align Torus With Axis=false
    Handle Radius=2
    Large Radius=10
    MaskSpawnGroup=false
    Module.Override Randomness=false
    Torus Origin:Vector3f=$Particles.OriginPoint
   )
   &DynamicMaterialParameters(
    Index 0 Param 2=&UniformRangedFloat(Minimum=0, Maximum=1)
    Index 0 Param 3=1
    Index 0 Param 4=&UniformRangedFloat(Minimum=1, Maximum=2)
    Module.Param0WriteEnabled=false
    Module.Param1WriteEnabled=true
    Module.Param2WriteEnabled=false
    Module.Param3WriteEnabled=true
   ) {module=DynamicMaterialParameters002}
   &AddVelocityFromPoint(
    Module.Use Falloff Distance=false
    Origin Offset=&Add_Vector(A=$Particles.OriginPoint, B=<0 0 -1>)
    Velocity Strength=20
   )
   /SetVariables(
    Particles.ParticleRandom:NiagaraFloat=&RandomRangeFloat(Minimum=0, Maximum=1)
    Particles.Velocity=&Multiply_Vector(A=$Particles.Velocity, B=<1 1 0.1>)
   ) {isSetParameters=true, module=SetVariables_#1}
  ]
  particleUpdateScript [
   &ParticleState()
   &SkeletalMeshLocation(
    Apply To Particle Position=false
    Bone Sampling Mode=::NewEnumerator4
    Inherit Velocity=false
    MaskSpawnGroup=false
    Skeletal Mesh=$User.SkeletalMesh
    Use Sampled Mesh Attribute as Ribbon ID=false
   )
   /SetVariables(Particles.BoneLocation:Vector3f=$Output.SkeletalMeshLocation.SampledPosition) {isSetParameters=true}
   &InheritVelocity(
    Module.Limit Speed=true
    Velocity Limit=8
    Velocity Scale=<-0.01 -0.01 0>
    Velocity Source=&TransformVector(Vector=$Engine.Owner.Velocity, Source Space=::World, Destination Space=::Simulation)
   )
   &AccelerationForce(Acceleration=&UniformRangedVector(Minimum=<0 0 50>, Maximum=<1 1 100>))
   &Drag(Use Linear Drag=true, Use Rotational Drag=false)
   &SolveForcesAndVelocity(Clamp Velocity=false, Limit Acceleration=false)
   &ScaleColor(
    Scale Mode=::NewEnumerator1
    Scale RGBA=&UniformRangedVector4(
     Minimum:Vector4f=&Vector4FromCurve(Vector4Curve=DataInterface() {propertyValues {
        AlphaCurve {keys [<at 0.0024 0>,<at 0.1271 1>,<at 0.265 1>,<at 0.9977 -0.0026>]}
        BlueCurve {keys [<at 0.2027 0.5>,<at 0.4709 0.1275>,<at 0.75 0>]}
        GreenCurve {keys [<at 0.2027 2.875>,<at 0.4709 0.7331>,<at 0.75 0>]}
        RedCurve {keys [<at 0.2027 10>,<at 0.4709 5>,<at 0.75 2>]}
       }})
     Maximum:Vector4f=&Vector4FromCurve(Vector4Curve=DataInterface() {propertyValues {
        AlphaCurve {keys [<at 0.0024 0>,<at 0.1271 1>,<at 0.265 0.9976>,<at 0.9977 -0.0026>]}
        BlueCurve {keys [<at 0.3937 0.5>,<at 0.75 0>]}
        GreenCurve {keys [<at 0.3937 2.875>,<at 0.75 0.5>]}
        RedCurve {keys [<at 0.3937 10>,<at 0.75 10>]}
       }})
     Randomness Mode:ENiagaraRandomnessMode=::NewEnumerator0
     Random Seed:NiagaraInt32=0
     Fixed Random Seed:NiagaraBool=false
    )
    ScaleRGBA=true
   ) {module=ScaleColor001}
   /Niagara/Modules/Update/SubUV/SubUVAnimation(Number Of Frames=64, SubUV Lookup Index=&Add_Float(A=$Particles.ParticleRandom, B=$Particles.Age))
   &ScaleSpriteSize(Scale Factor=&Vector2DFromFloat(Value=&FloatFromCurve(FloatCurve=DataInterface() {propertyValues {Curve {keys [
          {arriveTangent=7.5764, interpMode=RCIM_Cubic, leaveTangent=7.5765, tangentMode=RCTM_User, time=0, value=0}
          {arriveTangent=1.2063, interpMode=RCIM_Cubic, leaveTangent=1.2063, tangentMode=RCTM_User, time=0.18, value=0.99}
          <at 1 0.5 RCIM_Cubic>
         ]}}})))
   &CameraOffset(Camera Offset Amount=5)
   &DynamicMaterialParameters(
    Index 0 Param 1=&FloatFromCurve(FloatCurve=DataInterface() {propertyValues {Curve {keys [{arriveTangent=2.7551, interpMode=RCIM_Cubic, leaveTangent=2.7551, tangentMode=RCTM_User, time=0, value=0}, <at 0.7 1>]}}})
    Index 0 Param 2=1
    Index 0 Param 3=1
    Index 0 Param 4=1
    Module.Param0WriteEnabled=true
    Module.Param1WriteEnabled=true
    Module.Param2WriteEnabled=true
    Module.Param3WriteEnabled=false
   ) {module=DynamicMaterialParameters001}
   &KillParticlesInVolume(Invert Volume=true, Sphere Radius=40, Volume Origin:Vector3f=$Particles.OriginPoint)
   &self:TransformToBoneSpace(
    BonePosition:Vector3f=$Output.SkeletalMeshLocation.SampledPosition
    BoneRotation:Quat4f=$Output.SkeletalMeshLocation.SampledRotation
    Front End Toggle:NiagaraBool=$User.bIsFrontEnd
    Position:NiagaraPosition=$Particles.Position
   ) {module=TransformToBoneSpace}
  ]
  rendererClasses [&NiagaraSpriteRendererProperties]
  renderers [{rendererClass=&NiagaraSpriteRendererProperties}]
 }
 Light {
  emitterUpdateScript [
   &EmitterState()
   &SpawnBurst_Instantaneous(Spawn Count=1, Use Spawn Probability=false)
  ]
  particleSpawnScript [
   &InitializeParticle(
    Color=<rgba 6 1.8281 0.4375 0.1>
    Module.Write Color=true
    Module.Write Lifetime=true
    Module.Write Mass=false
    Module.Write Material Random=true
    Module.Write Position=false
    Module.Write Scale=false
    Module.Write SpriteRotation=false
    Module.Write SpriteSize=true
    Position=&SimulationPosition()
    Sprite Size=<60 40>
   )
   &DynamicMaterialParameters(
    Index 0 Param 1=-0.5
    Index 0 Param 2=50
    Module.Param0WriteEnabled=true
    Module.Param1WriteEnabled=true
    Module.Param2WriteEnabled=true
    Module.Param3WriteEnabled=true
   )
   &SystemLocation(MaskSpawnGroup=false, Offset=&FN_FrontEndVector3Modifier(Gameplay=<0 -9 10>, Front End=<0 -9 0>))
  ]
  particleUpdateScript [
   &SkeletalMeshLocation(
    Apply To Particle Position=false
    Bone Sampling Mode=::NewEnumerator4
    Inherit Velocity=false
    MaskSpawnGroup=false
    Skeletal Mesh=$User.SkeletalMesh
    Use Sampled Mesh Attribute as Ribbon ID=false
   )
   /SetVariables(Particles.BoneLocation:Vector3f=$Output.SkeletalMeshLocation.SampledPosition) {isSetParameters=true}
   /SetVariables(
    Particles.Light Color:LinearColor=&MultiplyLinearColor(
     Linear Color=<rgba 6 1.8281 0.4375 1>
     Scale Factor=&MakeVector4_FromFloat(Float=&FloatFromCurve(
       FloatCurve=DataInterface() {propertyValues {Curve {keys [<at 0 1>,<at 0.0584 1.1216>,<at 0.0797 0.6691>,<at 0.1464 0.526>,<at 0.222 1.0327>,<at 0.2705 0.6844>,<at 0.3166 0.8798>,<at 0.3641 0.5477>,<at 0.4407 0.6405>,<at 0.4815 1.1358>,<at 0.5388 1.0006>,<at 0.58 0.5314>,<at 0.6479 0.8407>,<at 0.6937 0.7459>,<at 0.737 1.0896>,<at 0.8125 0.5703>,<at 0.8909 0.8732>,<at 0.9449 0.6186>,<at 1 1>]}}}
       CurveIndex=$System.NormalizedLoopAge
      ))
    )
    Particles.LightExponent=1
    Particles.LightPosition:Vector3f=&Add_Vector(A=Linked() {linkedVariable {name=Particles.Position, type=NiagaraPosition}}, B=<0 0 10>)
    Particles.LightRadius=75
   ) {isSetParameters=true}
   &ParticleState()
   &SolveForcesAndVelocity(Clamp Velocity=false, Limit Acceleration=false)
   &self:TransformToBoneSpace(
    BonePosition:Vector3f=$Particles.BoneLocation
    BoneRotation:Quat4f=$Output.SkeletalMeshLocation.SampledRotation
    Front End Toggle:NiagaraBool=$User.bIsFrontEndPreview
    Position:Vector3f=$Particles.LightPosition
   ) {module=TransformToBoneSpace}
   /SetVariables(Particles.LightPosition:Vector3f=$Particles.OffsetPosition) {isSetParameters=true}
   &self:TransformToBoneSpace(
    BonePosition:Vector3f=$Particles.BoneLocation
    BoneRotation:Quat4f=$Output.SkeletalMeshLocation.SampledRotation
    Front End Toggle:NiagaraBool=$User.bIsFrontEnd
    Position:NiagaraPosition=$Particles.Position
   ) {module=TransformToBoneSpace001}
  ]
  rendererClasses [&NiagaraLightRendererProperties, &NiagaraSpriteRendererProperties]
  renderers [
   {rendererClass=&NiagaraLightRendererProperties, rendererIndex=0}
   {rendererClass=&NiagaraSpriteRendererProperties, rendererIndex=1}
  ]
 }
}
summary {
 systemName=NS_Glider_BabaYaga_Cauldron
 emitters [
  {emitterName=Bubbles, rendererClasses [&NiagaraSpriteRendererProperties, &NiagaraMeshRendererProperties]}
  {emitterName=Embers, rendererClasses [&NiagaraSpriteRendererProperties]}
  {emitterName=Light, rendererClasses [&NiagaraLightRendererProperties, &NiagaraSpriteRendererProperties]}
  {emitterName=Flames, rendererClasses [&NiagaraSpriteRendererProperties]}
  {emitterName=Fill, rendererClasses [&NiagaraSpriteRendererProperties]}
  {emitterName=Decal, rendererClasses [&NiagaraSpriteRendererProperties]}
  {emitterName=FlameLights, rendererClasses [&NiagaraLightRendererProperties]}
 ]
 userVariables [
  User.bIsFrontEnd {type=NiagaraBool}
  User.bIsFrontEndPreview {type=NiagaraBool}
  User.SkeletalMesh {
   type=NiagaraDataInterfaceSkeletalMesh
   defaultValue {
    struct=NiagaraExt_VariableValue_DataInterface
    value {dataInterface=&self:NiagaraDataInterfaceSkeletalMesh_0}
   }
  }
  User.TrailLengthScale {type=NiagaraFloat, defaultValue=0, description=""}
 ]
}
