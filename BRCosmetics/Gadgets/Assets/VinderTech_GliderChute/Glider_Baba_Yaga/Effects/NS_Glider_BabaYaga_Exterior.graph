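# Niagara system for the exterior effects of the Baba Yaga glider: twin mesh trails and a smoke exhaust attached to the glider's skeletal mesh.
# MeshTrails bursts once, follows two sampled sockets, and takes exhaust vector, length and opacity from system-level values.
# Smoke spawns sprites plus counter-ordered ribbons with eroding materials, by distance in gameplay.
# Front-end modifiers swap in alternate rates, offsets and velocity-derived exhaust for menu previews.
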NiagaraSystem /BRCosmetics/Gadgets/Assets/VinderTech_GliderChute/Glider_Baba_Yaga/Effects/NS_Glider_BabaYaga_Exterior
dependencies {
 usedDataInterfaces [/Script/Niagara.NiagaraDataInterfaceSimpleCounter, &NiagaraDataInterfaceCurve]
 usedDynamicInputs []
 usedModules [
  &SystemState
  &self:Set_Current_Exhaust
  &self:Set_Trail_Length
  &self:Set_Trail_Opacity
 ]
}
emitters {
 MeshTrails {
  emitterSpawnScript [&FN_PreserveSkydiveUserVariables() {module=PreserveUserVariables}]
  emitterUpdateScript [
   &EmitterState()
   &SpawnBurst_Instantaneous(Spawn Count=HlslExpression() {hlslExpression="Emitter.LoopCount == 0"}, Use Spawn Probability=false) {module=SpawnBurst_Instantaneous001}
  ]
  particleSpawnScript [
   /Niagara/Modules/Spawn/Initialization/InitializeParticle(
    Color=<rgba 1 1 1 0.025>
    Module.Write Color=true
    Module.Write Lifetime=true
    Module.Write Mass=false
    Module.Write Material Random=true
    Module.Write Position=false
    Module.Write Scale=false
    Module.Write SpriteRotation=false
    Module.Write SpriteSize=false
   )
   &self:Mesh_Trail_Setup(Trail Length:NiagaraFloat=&FN_FrontEndFloatModifier(Gameplay=150, Front End=200), Trail Width=75) {module=Mesh_Trail_Setup}
  ]
  particleUpdateScript [
   &ParticleState(Kill Particles When Lifetime Has Elapsed=false, Let Infinitely Lived Particles Die When Emitter Deactivates=true)
   &SkeletalMeshLocation(
    Apply Sampled Orientation to Mesh Particles=true
    Apply To Particle Position=false
    Bone / Socket Index=1
    Inherit Velocity=false
    Interpolate Mesh Sampling=false
    MaskSpawnGroup=false
    Mesh Sampling Type=::NewEnumerator5
    Sampled Position Offset Coordinate Space:ENiagaraMeshTransforms=::NewEnumerator3
    Skeletal Mesh=$User.SkeletalMesh
    Socket Sampling Mode=::NewEnumerator5
    Write Sampled Mesh Attrributes to Particle Attributes=true
   )
   &SkeletalMeshLocation(
    Apply Sampled Orientation to Mesh Particles=true
    Apply To Particle Position=false
    Bone / Socket Index=2
    Inherit Velocity=false
    Interpolate Mesh Sampling=false
    MaskSpawnGroup=false
    Mesh Sampling Type=::NewEnumerator5
    Sampled Position Offset Coordinate Space:ENiagaraMeshTransforms=::NewEnumerator3
    Skeletal Mesh=$User.SkeletalMesh
    Socket Sampling Mode=::NewEnumerator5
    Write Sampled Mesh Attrributes to Particle Attributes=true
   ) {module=SkeletalMeshLocation001}
   /SetVariables(
    Particles.TrailPosition:Vector3f=&Add_Vector(
     A=$Particles.SkeletalMeshLocation001.SampledPosition
     B=&FN_FrontEndVector3Modifier(
      Gameplay=<0 0 0>
      Front End=&RotateVectorByQuaternion(Vector To Rotate=<-5.1367 12.6744 0>, Quaternion=$Particles.SkeletalMeshLocation001.SampledRotation)
     )
    )
    Particles.TrailTwinPosition:Vector3f=&Add_Vector(
     A=$Particles.SkeletalMeshLocation.SampledPosition
     B=&FN_FrontEndVector3Modifier(
      Gameplay=<0 0 0>
      Front End=&RotateVectorByQuaternion(Vector To Rotate=<-5.1367 12.6744 0>, Quaternion=$Particles.SkeletalMeshLocation.SampledRotation)
     )
    )
   ) {isSetParameters=true}
   &self:Exhaust_to_DynaMat(
    ExhaustVector:Vector3f=&FN_FrontEndVector3Modifier(
     Gameplay=&Multiply_Vector(A=$System.CurrentExhaust, B=<1 6 1>)
     Front End=&self:Front_End_Exhaust(
      Rate Of Change:NiagaraFloat=10
      Velocity Input:Vector3f=&TransformVector(Vector=$Particles.SkeletalMeshLocation.SampledVelocity, Source Space=::World, Destination Space=::Local)
      Velocity Multiplier:Vector3f=<0 -7 0>
     )
    )
    SpeedMult=1
   ) {module=Exhaust_to_DynaMat}
   &self:Mesh_Trail_Length_Update(Dynamic Length Scale:NiagaraFloat=&FN_FrontEndFloatModifier(Gameplay=$System.TrailLength, Front End=1)) {module=Mesh_Trail_Length_Update}
   &ScaleColor(
    Scale Alpha=&FN_FrontEndFloatModifier(Gameplay=$System.TrailOpacity, Front End=1)
    ScaleA=true
    ScaleRGB=false
   )
  ]
  rendererClasses [&NiagaraMeshRendererProperties]
  renderers [
   {rendererClass=&NiagaraMeshRendererProperties, rendererIndex=0}
   {rendererClass=&NiagaraMeshRendererProperties, rendererIndex=1}
  ]
 }
 Smoke {
  emitterUpdateScript [
   &EmitterState()
   &SpawnPerUnit(
    Module.Use Max Movement Threshold=true
    Module.Use Movement Tolerance=false
    Spawn Per Unit:NiagaraFloat=&FN_FrontEndFloatModifier(Gameplay=30, Front End=0)
    Use Spawn Probability=false
   )
   &SpawnRate(SpawnRate=&FN_FrontEndFloatModifier(Gameplay=0, Front End=40), Use Spawn Probability=false)
  ]
  particleSpawnScript [
   &InitializeParticle(
    Color=&UniformRangedLinearColor(Minimum=<rgba 0.171 0.171 0.171 0.5>, Maximum=<rgba 0.228 0.228 0.228 0.95>)
    Lifetime=0.375
    Module.Write Color=true
    Module.Write Lifetime=true
    Module.Write Mass=true
    Module.Write Material Random=true
    Module.Write Position=false
    Module.Write Scale=false
    Module.Write SpriteRotation=true
    Module.Write SpriteSize=true
    Position=&SimulationPosition()
    Sprite Rotation=&UniformRangedFloat(Minimum=0, Maximum=360)
    Sprite Size=&Vector2DFromFloat(Value=&UniformRangedFloat(Minimum=10, Maximum=30))
   )
   &InitializeRibbon(
    Lifetime=0.3
    Module.Write Color=false
    Module.Write Lifetime=false
    Module.Write Mass=false
    Module.Write Material Random=true
    Module.Write Position=false
    Module.Write RibbonTwist=false
    Module.Write RibbonWidth=true
    Ribbon Width=50
   )
   &AccelerationForce(Acceleration=&FN_FrontEndVector3Modifier(
     Gameplay=&UniformRangedVector(Minimum=<0 0 0>, Maximum=<0 0 50>)
     Front End=&UniformRangedVector(Minimum=<0 0 100>, Maximum=<0 0 300>)
    ))
   &ApplyInitialForces()
   &SkeletalMeshLocation(
    Inherit Velocity=false
    MaskSpawnGroup=false
    Mesh Sampling Type=::NewEnumerator5
    Sampled Position Offset=&FN_FrontEndVector3Modifier(Gameplay=<0 0 0>, Front End=<0 0 -5.4658>)
    Skeletal Mesh=$User.SkeletalMesh
    Socket Sampling Mode=::NewEnumerator5
   )
   &AddVelocity(Velocity=&FN_FrontEndVector3Modifier(Gameplay=<0 -250 0>, Front End=<0 -500 0>))
   &InheritVelocity(
    Module.Limit Speed=false
    Velocity Scale=&VectorFromFloat(Value=&Lerp_Float(A=1, B=0, Alpha=$Particles.NormalizedAge))
   )
   /SetVariables(Particles.RibbonLinkOrder:NiagaraFloat=&MakeFloatFromInt(Int=/Niagara/DynamicInputs/DataInterface/GenerateIntFromCounter(SimpleCounter:NiagaraDataInterfaceSimpleCounter=DataInterface() {propertyValues {GpuSyncMode=None, InitialValue=0}}))) {isSetParameters=true}
  ]
  particleUpdateScript [
   &ParticleState()
   &AddVelocity(Velocity=&FN_FrontEndVector3Modifier(Gameplay=<0 0 0>, Front End=<0 50 0>)) {module=AddVelocity001}
   &Drag(Drag=3, Use Linear Drag=true, Use Rotational Drag=false)
   &SolveForcesAndVelocity(Clamp Velocity=false, Limit Acceleration=false)
   &ScaleSpriteSize(Scale Factor=&Vector2DFromFloat(Value=&FloatFromCurve(
      FloatCurve=DataInterface() {propertyValues {Curve {keys [{arriveTangent=2.1794, interpMode=RCIM_Cubic, leaveTangent=2.1794, tangentMode=RCTM_User, time=0, value=0}, <at 1 1 RCIM_Cubic>]}}}
      Scale Curve=2
     )))
   /SetVariables(Particles.SubImageIndex=&FloatFromCurve(FloatCurve=DataInterface() {propertyValues {Curve {keys [
         {arriveTangent=29.8623, interpMode=RCIM_Cubic, leaveTangent=29.8623, tangentMode=RCTM_User, time=-0.0046, value=5.0556}
         <at 1.0011 22.1689 RCIM_Cubic>
        ]}}})) {isSetParameters=true}
   &ScaleRibbonWidth(Ribbon Width Scale=&MakeCustomFloatFromBool(
     Bool=$User.bIsFrontEndPreview
     True Float=&FloatFromCurve(
      FloatCurve=DataInterface() {propertyValues {Curve {keys [
          {arriveTangent=4.1744, interpMode=RCIM_Cubic, leaveTangent=4.1744, tangentMode=RCTM_User, time=0, value=0}
          {interpMode=RCIM_Cubic, tangentMode=RCTM_User, time=0.6, value=1}
          {arriveTangent=-2.9855, interpMode=RCIM_Cubic, leaveTangent=-2.9855, tangentMode=RCTM_User, time=1.0043, value=0}
         ]}}}
      Scale Curve=0.5
     )
     False Float=&FloatFromCurve(FloatCurve=DataInterface() {propertyValues {Curve {keys [
          {arriveTangent=2.4416, interpMode=RCIM_Cubic, leaveTangent=2.4416, tangentMode=RCTM_User, time=0, value=0}
          {interpMode=RCIM_Cubic, tangentMode=RCTM_User, time=0.6, value=1}
         ]}}})
    ))
   /SetVariables(Particles.SpriteDynaMatParams:Vector4f=&MakeVector4(
     X=&FloatFromCurve(FloatCurve=DataInterface() {propertyValues {Curve {keys [<at -0.0046 1.0113 RCIM_Cubic>,<at 1.0011 0.002>]}}})
     Y=0
     Z=0
     W=&FloatFromCurve(FloatCurve=DataInterface() {propertyValues {Curve {keys [<at 0 0 RCIM_Cubic>,<at 1 0.5>]}}})
    )) {isSetParameters=true}
   /Game/Effects/Niagara/Modules/Materials/FN_RibbonMaterial/FN_RibbonMaterial(
    Color Pan U Speed=-1
    Erosion Amount:NiagaraFloat=&Lerp_Float(A=1, B=1.5, Alpha=$Particles.NormalizedAge)
    Erosion Pan Speed U=-2
    Jitter by Particles MaterialRandom=true
    Macro Distortion Amount=1
    Scale Alpha:NiagaraFloat=&FloatFromCurve(
     FloatCurve=DataInterface() {propertyValues {Curve {keys [
         {interpMode=RCIM_Cubic, tangentMode=RCTM_User, time=0, value=1}
         {arriveTangent=-2.8707, interpMode=RCIM_Cubic, leaveTangent=-2.8707, tangentMode=RCTM_User, time=1, value=0}
        ]}}}
     Scale Curve=2
    )
    TIme:NiagaraFloat=$Engine.Time
   )
   /SetVariables(Particles.RibbonColor:LinearColor=$Particles.Color) {isSetParameters=true}
   &ScaleColor(
    Scale Alpha=&FloatFromCurve(FloatCurve=DataInterface() {propertyValues {Curve {keys [
         {arriveTangent=-3.1757, interpMode=RCIM_Cubic, leaveTangent=-3.1757, tangentMode=RCTM_User, time=0.5, value=1.0113}
         {arriveTangent=-1.5121, interpMode=RCIM_Cubic, leaveTangent=-1.5121, tangentMode=RCTM_User, time=1.0011, value=0.002}
        ]}}})
    ScaleA=true
    ScaleRGB=false
   )
   &CameraOffset(Camera Offset Amount=3)
  ]
  rendererClasses [&NiagaraSpriteRendererProperties, &NiagaraRibbonRendererProperties]
  renderers [
   {rendererClass=&NiagaraSpriteRendererProperties}
   {rendererClass=&NiagaraRibbonRendererProperties, rendererIndex=1}
  ]
 }
}
summary {
 systemName=NS_Glider_BabaYaga_Exterior
 emitters [
  {emitterName=MeshTrails, rendererClasses [&NiagaraMeshRendererProperties]}
  {emitterName=Smoke, rendererClasses [&NiagaraSpriteRendererProperties, &NiagaraRibbonRendererProperties]}
 ]
 userVariables [
  User.bIsFrontEnd {type=NiagaraBool}
  User.bIsFrontEndPreview {type=NiagaraBool}
  User.BPOpacityMult {type=NiagaraFloat, defaultValue=0, description=""}
  User.CurrentExhaust {type=Vector3f, description="", defaultValue {struct=Vector3f, value=<0 0 0>}}
  User.ForwardDot {type=NiagaraFloat, defaultValue=0, description=""}
  User.PlayerSpeed {type=NiagaraFloat, defaultValue=0, description=""}
  User.RightDot {type=NiagaraFloat, defaultValue=0, description=""}
  User.SkeletalMesh {
   type=NiagaraDataInterfaceSkeletalMesh
   defaultValue {
    struct=NiagaraExt_VariableValue_DataInterface
    value {dataInterface=&self:NiagaraDataInterfaceSkeletalMesh_0}
   }
  }
  User.TrailLengthScale {type=NiagaraFloat, defaultValue=0, description=""}
 ]
}
